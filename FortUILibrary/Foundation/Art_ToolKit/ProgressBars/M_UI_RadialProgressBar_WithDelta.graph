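# Translucent UI material for a radial progress bar with a delta segment marking a change in progress.
# Radial coordinates from the UVs feed hardened progress and delta masks, positioned by animation scalars between start and stop values.
# Three-color gradients blend toward a supercharged palette, with separate delta colors and a position indicator.
# Circle functions shape the ring and end cap, while static switches toggle delta glow, the ring mask and sine-based segment separators.
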Material /FortUILibrary/Foundation/Art_ToolKit/ProgressBars/M_UI_RadialProgressBar_WithDelta
graph {
 properties {
  bUsedWithStaticMesh=false
  blendMode=BLEND_Translucent
  materialDomain=MD_UI
  pixelDepthOffsetMode=PDOM_Legacy
  usedShadingModels=Default Lit
 }
 nodes [
  n1 = NamedRerouteUsage()
  n2 = NamedRerouteUsage()
  n3 = Max(
   A=StaticSwitchParameter(True=NamedRerouteUsage(), False=NamedRerouteUsage()) {group=Appearance, parameterName=IsDeltaGlow}
   B=NamedRerouteUsage()
  )
  n4 = Multiply(A=Multiply(A=LinearInterpolate(A=@n1, B=@n2, Alpha=@n3), B=NamedRerouteUsage()), B=NamedRerouteUsage())
  n5 = NamedRerouteUsage()
  Add(A=@n4, B=@n5)
  n6 = NamedRerouteUsage()
  n7 = NamedRerouteUsage()
  n8 = NamedRerouteUsage()
  n9 = NamedRerouteUsage()
  n10 = NamedRerouteUsage()
  LinearInterpolate(
   A=LinearInterpolate(
    A=NamedRerouteUsage()
    B=LinearInterpolate(A=@n6, B=@n7, Alpha=@n8)
    Alpha=StaticSwitchParameter(True=@n9, False=@n10) {group=Appearance, parameterName=IsDeltaGlow}
   )
   Alpha=NamedRerouteUsage()
  )
  n11 = MaterialFunctionCall(UVs=NamedRerouteUsage(), Rotation Center=NamedRerouteUsage(), "Rotation Angle (0-1)"=NamedRerouteUsage()) {materialFunction=&CustomRotator}
  n12 = NamedRerouteUsage()
  LinearInterpolate(
   A=Multiply(A=NamedRerouteUsage()) {constB=0.5}
   Alpha=NamedRerouteDeclaration(None=MaterialFunctionCall(UVs=@n11.Rotated Values, Size=@n12) {materialFunction=&MF_UI_Circle}.Fill) {name=Circle, nodeColor=<rgba 0.8281 0 1 1>}
  )
  n13 = NamedRerouteUsage()
  Max(A=@n5, B=@n13)
  n14 = Multiply(A=NamedRerouteUsage(), B=@n4)
  n15 = NamedRerouteUsage()
  Multiply(
   A=StaticSwitchParameter(True=Power(Base=@n14, Exp=@n15), False=Max(A=@n4, B=@n13)) {group=Appearance, parameterName=IsDeltaGlow}
   B=NamedRerouteUsage()
  )
  NamedRerouteDeclaration(None=LinearInterpolate(A=NamedRerouteUsage(), B=NamedRerouteUsage(), Alpha=NamedRerouteUsage())) {name=Progress_Animation, nodeColor=<rgba 0 0 0 1>}
  NamedRerouteDeclaration(None=LinearInterpolate(A=NamedRerouteUsage(), B=NamedRerouteUsage(), Alpha=NamedRerouteUsage())) {name=Progress_Animation_Delta, nodeColor=<rgba 0 0 0 1>}
  NamedRerouteDeclaration(None=VectorParameter() {group=Color, parameterName=ColorMiddle_SuperCharged, defaultValue=<rgba 0 1 0 1>}.RGB) {name=ColorMiddle_SuperCharged, nodeColor=<rgba 0 0 0 1>}
  NamedRerouteDeclaration(None=VectorParameter() {group=Color, parameterName=ColorLeft_SuperCharged, defaultValue=<rgba 0 0 1 1>}.RGB) {name=ColorLeft_SuperCharged, nodeColor=<rgba 0 0 0 1>}
  NamedRerouteDeclaration(None=VectorParameter() {group=Color, parameterName=EmptyColor, defaultValue=<rgba 1 0 0 1>}.RGB) {name=EmptyColor, nodeColor=<rgba 0 0 0 1>}
  NamedRerouteDeclaration(None=ScalarParameter() {group=Appearance, parameterName=IsSuperCharged}) {name=IsSuperCharged, nodeColor=<rgba 0 0 0 1>}
  NamedRerouteDeclaration(None=ScalarParameter() {group=ProgressBar, parameterName=Progress_Start}) {name=Progress_Start, nodeColor=<rgba 0 0 0 1>}
  NamedRerouteDeclaration(None=ScalarParameter() {defaultValue=0.05, group=ProgressBar, parameterName=Delta_EdgeLocationCompenstation}) {name=Delta_EdgeLocationCompensation, nodeColor=<rgba 0 0 0 1>}
  NamedRerouteDeclaration(None=ScalarParameter() {defaultValue=1, group=ProgressBar, parameterName=Progress_Stop}) {name=Progress_Stop, nodeColor=<rgba 0 0 0 1>}
  NamedRerouteDeclaration(None=ScalarParameter() {defaultValue=0.666, group=Animate, parameterName=Animate_Delta}) {name=Animate_Delta, nodeColor=<rgba 0 0 0 1>}
  NamedRerouteDeclaration(None=@n14) {name=Delta Glow alpha, nodeColor=<rgba 0 0 0 1>}
  NamedRerouteDeclaration(None=ScalarParameter() {defaultValue=0.333, group=Animate, parameterName=Animate_Progress}) {name=Animate_Progress, nodeColor=<rgba 0 0 0 1>}
  n16 = NamedRerouteUsage()
  n17 = NamedRerouteUsage()
  NamedRerouteDeclaration(None=LinearInterpolate(A=NamedRerouteUsage(), B=Add(A=@n16, B=@n17), Alpha=NamedRerouteUsage())) {name=Progress_Animation_Delta_Compensated, nodeColor=<rgba 0 0 0 1>}
  NamedRerouteDeclaration(None=ScalarParameter() {defaultValue=12, group=Appearance, parameterName=Delta_EdgeHardness}) {name=Delta_EdgeHardness, nodeColor=<rgba 0 0 0 1>}
  NamedRerouteDeclaration(None=ScalarParameter() {defaultValue=80, group=Appearance, parameterName=Progress_EdgeHardness}) {name=Progress_EdgeHardness, nodeColor=<rgba 0 0 0 1>}
  NamedRerouteDeclaration(None=ScalarParameter() {defaultValue=0.025, group=Mask, parameterName=Mask Value}) {name=Mask Value, nodeColor=<rgba 0 0 0 1>}
  NamedRerouteDeclaration(None=ScalarParameter() {defaultValue=40, group=Mask, parameterName=Mask SharpnessTopBottom}) {name=Mask SharpnessTopBottom, nodeColor=<rgba 0 0 0 1>}
  NamedRerouteDeclaration(None=ScalarParameter() {defaultValue=40, group=Mask, parameterName=Mask SharpnessRightLeft}) {name=Mask SharpnessRightLeft, nodeColor=<rgba 0 0 0 1>}
  n18 = TextureCoordinate()
  NamedRerouteDeclaration(None=@n18) {name=CoordUV, nodeColor=<rgba 0 0 0 1>}
  n19 = ComponentMask(None=OneMinus(None=MaterialFunctionCall!(Vector or UVs=@n18) {materialFunction=/FortUILibrary/MaterialFunctions/Basics/MF_UI_CheaperVectorToRadial}.Radial Coordinates)) {r=true}
  NamedRerouteDeclaration(None=@n19) {name=CoordU, nodeColor=<rgba 0 0 0 1>}
  NamedRerouteDeclaration(None=OneMinus(None=@n19)) {name=CoordU Inverted, nodeColor=<rgba 0 0 0 1>}
  NamedRerouteDeclaration(None=Multiply(A=@n19) {constB=-1}) {name=CoordU Negative, nodeColor=<rgba 0 0 0 1>}
  n20 = NamedRerouteUsage()
  n21 = Constant() {r=1}
  n22 = NamedRerouteUsage()
  n23 = Constant() {r=-1}
  NamedRerouteDeclaration(None=StaticSwitchParameter(
    True=MaterialFunctionCall(UVs=@n20, Size=@n21, OutlineThickness=@n22, OutlineAlignment=@n23) {materialFunction=&MF_UI_Circle}.Outline
    False=Constant() {r=1}
   ) {defaultValue=true, group=Appearance, parameterName=UseMaskForGlow}) {name=Mask, nodeColor=<rgba 0 0 0 1>}
  n24 = Add(A=NamedRerouteUsage(), B=NamedRerouteUsage())
  n25 = NamedRerouteUsage()
  NamedRerouteDeclaration(None=Saturate(None=Multiply(A=@n24, B=@n25))) {name=Progress, nodeColor=<rgba 0 0 0 1>}
  n26 = NamedRerouteUsage()
  n27 = Multiply(A=Add(A=NamedRerouteUsage(), B=NamedRerouteUsage()) {constB=0}, B=Multiply(A=NamedRerouteUsage()) {constB=0.5}) {constB=10}
  n28 = Subtract(A=Abs(None=Sine(None=@n27)), B=NamedRerouteUsage())
  n29 = Saturate(None=Multiply(A=Saturate(None=@n28), B=NamedRerouteUsage()) {constB=256})
  NamedRerouteDeclaration(None=StaticSwitchParameter(True=LinearInterpolate(A=@n26, Alpha=@n29), False=Constant() {r=1}) {group=Appearance, parameterName=UseSeparatorMask}) {name=separator, nodeColor=<rgba 0 0 0 1>}
  n30 = NamedRerouteUsage()
  n31 = NamedRerouteUsage()
  n32 = NamedRerouteUsage()
  n33 = Power(Base=NamedRerouteUsage(), Exp=NamedRerouteUsage()) {constExponent=1.22}
  n34 = NamedRerouteUsage()
  n35 = NamedRerouteUsage()
  n36 = NamedRerouteUsage()
  NamedRerouteDeclaration(None=LinearInterpolate(
    A=MaterialFunctionCall(A=@n30, B=@n31, C=@n32, Alpha=@n33) {materialFunction=&3ColorBlend}.Result
    B=MaterialFunctionCall(A=@n34, B=@n35, C=@n36, Alpha=@n33) {materialFunction=&3ColorBlend}.Result
    Alpha=NamedRerouteUsage()
   )) {name=Progress Color, nodeColor=<rgba 0 0 0 1>}
  NamedRerouteDeclaration(None=ScalarParameter() {defaultValue=1.22, group=Appearance, parameterName=TaperorFlattenGradient}) {name=TaperorFlattenGradient, nodeColor=<rgba 0 0 0 1>}
  NamedRerouteDeclaration(None=ScalarParameter() {defaultValue=0.5, group=Animate, parameterName=Animate_DeltaColor, sliderMax=1}) {name=Animate_DeltaColor, nodeColor=<rgba 0 0 0 1>}
  NamedRerouteDeclaration(None=VectorParameter() {group=Color, parameterName=DeltaProgressColor, defaultValue=<rgba 1 0 1 1>}.RGB) {name=DeltaProgressColor, nodeColor=<rgba 0 0 0 1>}
  NamedRerouteDeclaration(None=VectorParameter() {group=Color, parameterName=DeltaProgressColorDecay, defaultValue=<rgba 1 0.0223 0.6569 1>}.RGB) {name=DeltaProgressColorDecay, nodeColor=<rgba 0 0 0 1>}
  NamedRerouteDeclaration(None=ScalarParameter() {defaultValue=40, group=Appearance, parameterName=Progress_IndicatorSize}) {name=Progress_IndicaterSize, nodeColor=<rgba 0 0 0 1>}
  NamedRerouteDeclaration(None=ScalarParameter() {defaultValue=50, group=Appearance, parameterName=PositionIndicatorIntensity}) {name=PositionIndicatorIntensity, nodeColor=<rgba 0 0 0 1>}
  n37 = NamedRerouteUsage()
  n38 = Power(Base=Add(A=NamedRerouteUsage(), B=OneMinus(None=@n37)), Exp=NamedRerouteUsage()) {constExponent=32}
  NamedRerouteDeclaration(None=Multiply(A=Saturate(None=@n38), B=NamedRerouteUsage())) {name=ProgressIndicator, nodeColor=<rgba 0 0 0 1>}
  NamedRerouteDeclaration(None=ScalarParameter() {group=Appearance, parameterName=EmptyAlpha, sliderMax=1}) {name=EmptyAlpha, nodeColor=<rgba 0 0 0 1>}
  n39 = Multiply(A=Add(A=NamedRerouteUsage(), B=NamedRerouteUsage()), B=NamedRerouteUsage())
  n40 = NamedRerouteUsage()
  NamedRerouteDeclaration(None=Multiply(A=Saturate(None=@n39), B=OneMinus(None=@n40))) {name=Delta, nodeColor=<rgba 0 0 0 1>}
  NamedRerouteDeclaration(None=ScalarParameter() {defaultValue=1, group=Appearance, parameterName=FillAlpha, sliderMax=1}) {name=FillAlpha, nodeColor=<rgba 0 0 0 1>}
  NamedRerouteDeclaration(None=ScalarParameter() {defaultValue=1.22, group=Appearance, parameterName=TaperDeltaGlow}) {name=TaperDeltaGlow, nodeColor=<rgba 0 0 0 1>}
  NamedRerouteDeclaration(None=ScalarParameter() {defaultValue=1, group=Appearance, parameterName=FinalAlpha, sliderMax=1}) {name=FinalAlpha, nodeColor=<rgba 0 0 0 1>}
  n41 = NamedRerouteUsage()
  n42 = NamedRerouteUsage()
  n43 = Saturate(None=Add(A=NamedRerouteUsage(), B=Multiply(A=@n42) {constB=0.01}))
  n44 = NamedRerouteUsage()
  NamedRerouteDeclaration(None=LinearInterpolate(A=NamedRerouteUsage(), B=LinearInterpolate(A=@n41, B=@n43, Alpha=@n44), Alpha=NamedRerouteUsage())) {name=Progress with indicator, nodeColor=<rgba 0 0 0 1>}
  NamedRerouteDeclaration(None=ScalarParameter() {group=Separator, parameterName=SeparatorPosition}) {name=SeparatorPosition, nodeColor=<rgba 0 0 0 1>}
  NamedRerouteDeclaration(None=ScalarParameter() {defaultValue=10, group=Separator, parameterName=Separator Number}) {name=Separator Number, nodeColor=<rgba 0 0 0 1>}
  NamedRerouteDeclaration(None=ScalarParameter() {defaultValue=0.1, group=Separator, parameterName=Separators_Thickness}) {name=Separator_Thickness, nodeColor=<rgba 0 0 0 1>}
  NamedRerouteDeclaration(None=ScalarParameter() {defaultValue=128, group=Separator, parameterName=Separator_Sharpness}) {name=Separator_Sharpness, nodeColor=<rgba 0 0 0 1>}
  NamedRerouteDeclaration(None=ScalarParameter() {group=Separator, parameterName=LiftSegmentAlpha}) {name=LiftSegmentAlpha, nodeColor=<rgba 0 0 0 1>}
  n45 = Multiply(A=Add(A=NamedRerouteUsage(), B=NamedRerouteUsage()), B=NamedRerouteUsage())
  n46 = NamedRerouteUsage()
  NamedRerouteDeclaration(None=Multiply(A=Saturate(None=@n45), B=OneMinus(None=@n46))) {name=Delta compensated, nodeColor=<rgba 0 0 0 1>}
  n47 = NamedRerouteUsage()
  n48 = ComponentMask(None=@n47) {g=true}
  n49 = Subtract(A=Constant() {r=0.5}, B=Multiply(A=NamedRerouteUsage()) {constB=0.5})
  NamedRerouteDeclaration(None=AppendVector(A=ComponentMask(None=@n47) {r=true}, B=Add(A=@n48, B=@n49))) {name=CornerUV, nodeColor=<rgba 0 0 0 1>}
  NamedRerouteDeclaration(None=ScalarParameter() {defaultValue=0.1, parameterName=Mask Thickness}) {name=Mask Thickness, nodeColor=<rgba 0 0 0 1>}
  NamedRerouteDeclaration(None=MaterialFunctionCall(UVs=@n11.Rotated Values, Size=@n12, GlowMax=Constant() {r=0.02}) {materialFunction=&MF_UI_Circle}.Fill) {name=CircleGlowMask, nodeColor=<rgba 0.5234 1 0 1>}
  NamedRerouteDeclaration(None=VectorParameter() {group=Color, parameterName=ColorRight, defaultValue=<rgba 1 0 0 1>}.RGB) {name=ColorRight, nodeColor=<rgba 0 0 0 1>}
  NamedRerouteDeclaration(None=VectorParameter() {group=Color, parameterName=ColorMiddle, defaultValue=<rgba 0 1 0 1>}.RGB) {name=ColorMiddle, nodeColor=<rgba 0 0 0 1>}
  NamedRerouteDeclaration(None=VectorParameter() {group=Color, parameterName=ColorLeft, defaultValue=<rgba 0 0 1 1>}.RGB) {name=ColorLeft, nodeColor=<rgba 0 0 0 1>}
  NamedRerouteDeclaration(None=VectorParameter() {group=Color, parameterName=ColorRight_SuperCharged, defaultValue=<rgba 1 0 0 1>}.RGB) {name=ColorRight_SuperCharged, nodeColor=<rgba 0 0 0 1>}
  NamedRerouteUsage()
 ]
}
